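# MetaSound patch that compares float A against float B using a selectable comparison type and outputs a Bool Result.
# A change in A fires a Trigger Compare node, so changes to B or Type alone do not re-evaluate it.
# The compare's TRUE and FALSE triggers drive a two-way Bool trigger route, turning the trigger outcome into a held value.
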MetaSoundPatch /Game/Patches/CompareFloat
inputs [
 A {type=Float}
 B {type=Float}
 Type {type=Enum:TriggerComparisonType}
]
interfaces [MetaSound:1.0]
outputs [Result {type=Bool}]
pages [Default {
  input_defaults {A=0, B=0, Type=5}
  nodes [
   InitVariable.Float()
   TriggerOnValueChange_0 = UE.Trigger On Value Change:Float(Value=^A)
   n1 = TriggerCompare.Clamp:Float(B=^B, Type=^Type, A=UE.Reroute(Value=UE.Reroute(Value=^A).Value).Value, Compare=@TriggerOnValueChange_0.Trigger)
   TriggerRouteTriggerRouteBool2_0 = "TriggerRoute.Trigger Route (Bool, 2):Bool"(Set 1=@n1.FALSE, Set 0=@n1.TRUE) {also {defaults {Value 0="<MetaSoundBuilderResult.SUCCEEDED: 0>"}}}
  ]
  links [{dst="@Result", src=TriggerRouteTriggerRouteBool2_0.Value}]
 }]
