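# MetaSound patch that checks whether a WaveAsset array is valid, routing an incoming trigger to either Is Valid or Is Not Valid.
# The array's element count feeds an Int32 Trigger Compare left at its default comparison settings, evaluated when Trigger In fires.
# The true result fires Is Not Valid and false fires Is Valid, so validity rests on the count alone, not on individual entries.
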MetaSoundPatch /Game/Patches/IsWaveAssetArrayValid
inputs [
 Trigger In {type=Trigger}
 WaveAsset {type=WaveAsset:Array}
]
interfaces [MetaSound:1.0]
outputs [
 Is Not Valid {type=Trigger}
 Is Valid {type=Trigger}
]
pages [Default {
  input_defaults {Trigger In=null, WaveAsset []}
  nodes [TriggerCompareClamp_0 = TriggerCompare.Clamp:Int32(Compare=^Trigger In, A=Array.Num:WaveAsset:Array(Array=^WaveAsset).Num)]
  links [{dst="@Is Valid", src=TriggerCompareClamp_0.FALSE}, {dst="@Is Not Valid", src=TriggerCompareClamp_0.TRUE}]
 }]
